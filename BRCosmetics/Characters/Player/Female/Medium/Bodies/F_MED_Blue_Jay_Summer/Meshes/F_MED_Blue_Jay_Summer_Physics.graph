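# Physics asset for the F_MED_Blue_Jay_Summer female player body cosmetic, set up for a dangling belt.
# Six short capsules, mostly thickening toward the tip, form a dyn_belt bone chain constrained to the pelvis.
# Pelvis (capsule plus large box) and left thigh (capsule) are the only other bodies, giving the belt shapes to rest against.
# Belt joints swing freely, with twist locked at the pelvis attachment and the last two joints.
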PhysicsAsset /BRCosmetics/Characters/Player/Female/Medium/Bodies/F_MED_Blue_Jay_Summer/Meshes/F_MED_Blue_Jay_Summer_Physics
bodies {
 dyn_belt_1 [{shapeType=Capsule, length=4, radius=1, shapeName=dyn_belt_3_capsule, center=<2 0 0>, rotation=<rot 90 0 0>}]
 dyn_belt_2 [{shapeType=Capsule, length=6, radius=1, shapeName=dyn_belt_3_capsule, center=<3 0 0>, rotation=<rot 90 0 0>}]
 dyn_belt_3 [{shapeType=Capsule, length=6, radius=1.5, shapeName=dyn_belt_3_capsule, center=<3 0 0>, rotation=<rot 90 0 0>}]
 dyn_belt_4 [{shapeType=Capsule, length=6, radius=1.5, shapeName=dyn_belt_4_capsule, center=<3 0 0>, rotation=<rot 90 0 0>}]
 dyn_belt_5 [{shapeType=Capsule, length=6, radius=1.5, shapeName=dyn_belt_5_capsule, center=<3 0 0>, rotation=<rot 90 0 0>}]
 dyn_belt_6 [{shapeType=Capsule, length=6, radius=1.9, shapeName=dyn_belt_6_capsule, center=<3 0 0>, rotation=<rot 90 0 0>}]
 pelvis [
  {shapeType=Capsule, length=7, radius=12, shapeName=pelvis_capsule, center=<-1.5498 -2.7122 0.209>}
  {shapeType=Box, extentX=60, extentY=52, extentZ=40, shapeName=pelvis_box, center=<-14 10 10>, rotation=<rot 4 0 0>}
 ]
 thigh_l [{shapeType=Capsule, length=38.25, radius=7.1, shapeName=thigh_l_capsule, center=<-15.4282 2.8466 0>, rotation=<rot 90 180 -172.7377>}]
}
constraints [
 {
  bone1Name=thigh_l
  bone2Name=pelvis
  swing1LimitDegrees=45
  swing1Motion=Free
  swing2LimitDegrees=45
  swing2Motion=Free
  twistLimitDegrees=45
  twistMotion=Free
 }
 {
  bone1Name=dyn_belt_1
  bone2Name=pelvis
  swing1LimitDegrees=45
  swing1Motion=Free
  swing2LimitDegrees=45
  swing2Motion=Free
  twistLimitDegrees=5
  twistMotion=Locked
 }
 {
  bone1Name=dyn_belt_2
  bone2Name=dyn_belt_1
  swing1LimitDegrees=45
  swing1Motion=Free
  swing2LimitDegrees=45
  swing2Motion=Free
  twistLimitDegrees=5
 }
 {
  bone1Name=dyn_belt_3
  bone2Name=dyn_belt_2
  swing1LimitDegrees=45
  swing1Motion=Free
  swing2LimitDegrees=45
  swing2Motion=Free
  twistLimitDegrees=5
 }
 {
  bone1Name=dyn_belt_4
  bone2Name=dyn_belt_3
  swing1LimitDegrees=45
  swing1Motion=Free
  swing2LimitDegrees=45
  swing2Motion=Free
  twistLimitDegrees=5
 }
 {
  bone1Name=dyn_belt_5
  bone2Name=dyn_belt_4
  swing1LimitDegrees=45
  swing1Motion=Free
  swing2LimitDegrees=45
  swing2Motion=Free
  twistLimitDegrees=5
  twistMotion=Locked
 }
 {
  bone1Name=dyn_belt_6
  bone2Name=dyn_belt_5
  swing1LimitDegrees=45
  swing1Motion=Free
  swing2LimitDegrees=45
  swing2Motion=Free
  twistLimitDegrees=5
  twistMotion=Locked
 }
]
