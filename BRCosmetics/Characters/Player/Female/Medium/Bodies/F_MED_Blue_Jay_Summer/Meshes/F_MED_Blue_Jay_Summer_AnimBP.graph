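# Animation Blueprint for the female Blue Jay Summer body cosmetic, adding secondary motion to the attached parent's pose.
# The copied pose feeds a Control Rig, AnimDynamics for bracelets and shoelaces, then a RigidBody simulation preceded by a toggleable belt rotation.
# Final bone offsets shift the left bracelet during the menu idle montage and the backpack when tagged EnsembleMaroon.
# Each update refreshes those flags and RigidBody gravity, using a deprecated state-based override with skydive values.
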AnimBlueprint /BRCosmetics/Characters/Player/Female/Medium/Bodies/F_MED_Blue_Jay_Summer/Meshes/F_MED_Blue_Jay_Summer_AnimBP
target_skeleton=&/F_MED_Blue_Jay_Summer_Skeleton
graphs [
 AnimGraph {nodes [
   n1 = Control Rig(Source=Copy Pose From Mesh() {bUseAttachedParent=true}) {
    alphaBoolBlend {blendOption=Linear}
    controlRigAssetReference {blueprintRigClass=&/F_MED_Blue_Jay_Summer_ControlRig_new.F_MED_Blue_Jay_Summer_ControlRig_new_C}
   }
   n2 = Anim Dynamics - Bone: dyn_L_bracelet(ComponentPose=Local To Component(LocalPose=@n1), Alpha=1, ExternalForce="0,0,0") {
    angularDampingOverride=1
    angularSpringConstant=200000
    bLinearSpring=true
    bOverrideLinearDamping=true
    bUseGravityOverride=true
    gravityScale=0.35
    linearDampingOverride=1
    linearSpringConstant=50
    simulationSpace=BoneRelative
    alphaBoolBlend {blendOption=Linear}
    boundBone {boneName=dyn_L_bracelet}
    gravityOverride {z=-980}
    physicsBodyDefinitions [{
      boundBone {boneName=dyn_L_bracelet}
      boxExtents=<200 200 200>
      constraintSetup {angularLimitsMax=<0 0 0>, angularLimitsMin=<0 0 0>, angularTarget=<0 0 0>, linearAxesMax=<1.5 0 0>, linearAxesMin=<-2.5 0 0>}
      localJointOffset=<0.5 0 0>
     }]
    relativeSpaceBone {boneName=clavicle_l}
   }
   n3 = Anim Dynamics - Bone: dyn_R_bracelet(ComponentPose=@n2, Alpha=1, ExternalForce="0,0,0") {
    angularDampingOverride=1
    angularSpringConstant=200000
    bLinearSpring=true
    bOverrideLinearDamping=true
    bUseGravityOverride=true
    gravityScale=0.35
    linearDampingOverride=1
    linearSpringConstant=50
    simulationSpace=BoneRelative
    alphaBoolBlend {blendOption=Linear}
    boundBone {boneName=dyn_R_bracelet}
    gravityOverride {z=-980}
    physicsBodyDefinitions [{
      boundBone {boneName=dyn_R_bracelet}
      boxExtents=<200 200 200>
      constraintSetup {angularLimitsMax=<0 0 0>, angularLimitsMin=<0 0 0>, angularTarget=<0 0 0>, linearAxesMax=<2.5 0 0>, linearAxesMin=<-1.5 0 0>}
      localJointOffset=<-0.5 0 0>
     }]
    relativeSpaceBone {boneName=clavicle_r}
   }
   n4 = Anim Dynamics - Bone: dyn_L_shoelace_A1(ComponentPose=@n3, Alpha=1, ExternalForce="0,0,0") {
    angularDampingOverride=1
    angularSpringConstant=20000
    bAngularSpring=true
    bOverrideAngularDamping=true
    bUseGravityOverride=true
    gravityScale=0.35
    simulationSpace=RootRelative
    alphaBoolBlend {blendOption=Linear}
    boundBone {boneName=dyn_L_shoelace_A1}
    componentLinearAccScale=<1 1 1>
    componentLinearVelScale=<1 1 1>
    gravityOverride {z=-980}
    physicsBodyDefinitions [{
      boundBone {boneName=dyn_L_shoelace_A1}
      boxExtents=<10 10 10>
      constraintSetup {angularLimitsMax=<20 10 20>, angularLimitsMin=<-20 -10 -20>, angularTarget=<0 0 0>}
      localJointOffset=<2 0 0>
     }]
   }
   n5 = Anim Dynamics - Bone: dyn_L_shoelace_A2(ComponentPose=@n4, Alpha=1, ExternalForce="0,0,0") {
    angularDampingOverride=1
    angularSpringConstant=20000
    bAngularSpring=true
    bOverrideAngularDamping=true
    bUseGravityOverride=true
    gravityScale=0.35
    simulationSpace=RootRelative
    alphaBoolBlend {blendOption=Linear}
    boundBone {boneName=dyn_L_shoelace_A2}
    componentLinearAccScale=<1 1 1>
    componentLinearVelScale=<1 1 1>
    gravityOverride {z=-980}
    physicsBodyDefinitions [{
      boundBone {boneName=dyn_L_shoelace_A2}
      boxExtents=<10 10 10>
      constraintSetup {angularLimitsMax=<20 10 20>, angularLimitsMin=<-20 -10 -20>, angularTarget=<0 0 0>}
      localJointOffset=<2 0 0>
     }]
   }
   n6 = Anim Dynamics - Bone: dyn_L_shoelace_B1(ComponentPose=@n5, Alpha=1, ExternalForce="0,0,0") {
    angularDampingOverride=1
    angularSpringConstant=20000
    bAngularSpring=true
    bOverrideAngularDamping=true
    bUseGravityOverride=true
    gravityScale=0.35
    simulationSpace=RootRelative
    alphaBoolBlend {blendOption=Linear}
    boundBone {boneName=dyn_L_shoelace_B1}
    componentLinearAccScale=<1 1 1>
    componentLinearVelScale=<1 1 1>
    gravityOverride {z=-980}
    physicsBodyDefinitions [{
      boundBone {boneName=dyn_L_shoelace_B1}
      boxExtents=<10 10 10>
      constraintSetup {angularLimitsMax=<20 10 20>, angularLimitsMin=<-20 -10 -20>, angularTarget=<0 0 0>}
      localJointOffset=<2 0 0>
     }]
   }
   n7 = Anim Dynamics - Bone: dyn_L_shoelace_B2(ComponentPose=@n6, Alpha=1, ExternalForce="0,0,0") {
    angularDampingOverride=1
    angularSpringConstant=20000
    bAngularSpring=true
    bOverrideAngularDamping=true
    bUseGravityOverride=true
    gravityScale=0.35
    simulationSpace=RootRelative
    alphaBoolBlend {blendOption=Linear}
    boundBone {boneName=dyn_L_shoelace_B2}
    componentLinearAccScale=<1 1 1>
    componentLinearVelScale=<1 1 1>
    gravityOverride {z=-980}
    physicsBodyDefinitions [{
      boundBone {boneName=dyn_L_shoelace_B2}
      boxExtents=<10 10 10>
      constraintSetup {angularLimitsMax=<20 10 20>, angularLimitsMin=<-20 -10 -20>, angularTarget=<0 0 0>}
      localJointOffset=<2 0 0>
     }]
   }
   n8 = Anim Dynamics - Bone: dyn_R_shoelace_A1(ComponentPose=@n7, Alpha=1, ExternalForce="0,0,0") {
    angularDampingOverride=1
    angularSpringConstant=20000
    bAngularSpring=true
    bOverrideAngularDamping=true
    bUseGravityOverride=true
    gravityScale=0.35
    simulationSpace=RootRelative
    alphaBoolBlend {blendOption=Linear}
    boundBone {boneName=dyn_R_shoelace_A1}
    componentLinearAccScale=<1 1 1>
    componentLinearVelScale=<1 1 1>
    gravityOverride {z=-980}
    physicsBodyDefinitions [{
      boundBone {boneName=dyn_R_shoelace_A1}
      boxExtents=<10 10 10>
      constraintSetup {angularLimitsMax=<20 10 20>, angularLimitsMin=<-20 -10 -20>, angularTarget=<0 0 0>}
      localJointOffset=<-2 0 0>
     }]
   }
   n9 = Anim Dynamics - Bone: dyn_R_shoelace_A2(ComponentPose=@n8, Alpha=1, ExternalForce="0,0,0") {
    angularDampingOverride=1
    angularSpringConstant=20000
    bAngularSpring=true
    bOverrideAngularDamping=true
    bUseGravityOverride=true
    gravityScale=0.35
    simulationSpace=RootRelative
    alphaBoolBlend {blendOption=Linear}
    boundBone {boneName=dyn_R_shoelace_A2}
    componentLinearAccScale=<1 1 1>
    componentLinearVelScale=<1 1 1>
    gravityOverride {z=-980}
    physicsBodyDefinitions [{
      boundBone {boneName=dyn_R_shoelace_A2}
      boxExtents=<10 10 10>
      constraintSetup {angularLimitsMax=<20 10 20>, angularLimitsMin=<-20 -10 -20>, angularTarget=<0 0 0>}
      localJointOffset=<-2 0 0>
     }]
   }
   n10 = Anim Dynamics - Bone: dyn_R_shoelace_B1(ComponentPose=@n9, Alpha=1, ExternalForce="0,0,0") {
    angularDampingOverride=1
    angularSpringConstant=20000
    bAngularSpring=true
    bOverrideAngularDamping=true
    bUseGravityOverride=true
    gravityScale=0.35
    simulationSpace=RootRelative
    alphaBoolBlend {blendOption=Linear}
    boundBone {boneName=dyn_R_shoelace_B1}
    componentLinearAccScale=<1 1 1>
    componentLinearVelScale=<1 1 1>
    gravityOverride {z=-980}
    physicsBodyDefinitions [{
      boundBone {boneName=dyn_R_shoelace_B1}
      boxExtents=<10 10 10>
      constraintSetup {angularLimitsMax=<20 10 20>, angularLimitsMin=<-20 -10 -20>, angularTarget=<0 0 0>}
      localJointOffset=<-2 0 0>
     }]
   }
   n11 = Anim Dynamics - Bone: dyn_R_shoelace_B2(ComponentPose=@n10, Alpha=1, ExternalForce="0,0,0") {
    angularDampingOverride=1
    angularSpringConstant=20000
    bAngularSpring=true
    bOverrideAngularDamping=true
    bUseGravityOverride=true
    gravityScale=0.35
    simulationSpace=RootRelative
    alphaBoolBlend {blendOption=Linear}
    boundBone {boneName=dyn_R_shoelace_B2}
    componentLinearAccScale=<1 1 1>
    componentLinearVelScale=<1 1 1>
    gravityOverride {z=-980}
    physicsBodyDefinitions [{
      boundBone {boneName=dyn_R_shoelace_B2}
      boxExtents=<10 10 10>
      constraintSetup {angularLimitsMax=<20 10 20>, angularLimitsMin=<-20 -10 -20>, angularTarget=<0 0 0>}
      localJointOffset=<-2 0 0>
     }]
   }
   n12 = "Transform (Modify) Bone - Bone: dyn_belt_master"!(ComponentPose=@n11, bAlphaBoolEnabled:bool=Get bRigidBodyEnabled(), Translation="0,0,0", Rotation="60,0,0", Scale="1,1,1") -> (Pose) {
    alphaInputType=Bool
    rotationMode=BMM_Additive
    rotationSpace=BCS_BoneSpace
    alphaBoolBlend {blendOption=Linear}
    boneToModify {boneName=dyn_belt_master}
    rotation {pitch=60}
   }
   n13 = RigidBody!(ComponentPose=@n12, Alpha, OverrideWorldGravity:Vector=Get gravityVector() -> (gravityVector:Vector), ExternalForce) -> (Pose) {
    bClampLinearTranslationLimitToRefPose=true
    bOverrideWorldGravity=true
    overridePhysicsAsset=&/F_MED_Blue_Jay_Summer_Physics
    simulationSpace=BaseBoneSpace
    alphaBoolBlend {blendOption=Linear}
    baseBoneRef {boneName=pelvis}
    componentAppliedLinearAccClamp=<3000 3000 3000>
    componentLinearAccScale=<0.2 0.2 0.2>
    componentLinearVelScale=<0.3 0.3 0.3>
    overrideWorldGravity {z=-980}
   }
   n14 = "Transform (Modify) Bone - Bone: dyn_L_bracelet"(ComponentPose=@n13, Alpha=Get idlePlaying() -> (idlePlaying:double), Translation="-4,0,0", Rotation="0,0,0", Scale="1,1,1") {
    translationMode=BMM_Additive
    translationSpace=BCS_BoneSpace
    alphaBoolBlend {blendOption=Linear}
    boneToModify {boneName=dyn_L_bracelet}
    translation {x=-4}
   }
   n15 = "Transform (Modify) Bone - Bone: dyn_backpack"!(
    ComponentPose=@n14
    bAlphaBoolEnabled:bool=Get EnsembleBackpack() -> (EnsembleBackpack:bool)
    Translation="12,0,0"
    Rotation="0,0,0"
    Scale="1,1,1"
   ) -> (Pose) {
    alphaInputType=Bool
    rotationMode=BMM_Additive
    rotationSpace=BCS_BoneSpace
    scaleMode=BMM_Additive
    scaleSpace=BCS_BoneSpace
    translationMode=BMM_Additive
    translationSpace=BCS_BoneSpace
    alphaBoolBlend {blendOption=Linear}
    boneToModify {boneName=dyn_backpack}
    translation {x=12}
   }
   Output Pose(Result=Component To Local(ComponentPose=@n15)) {layerGroup=None}
  ]}
 EventGraph {nodes [
   Event BlueprintInitializeAnimation()
   > n1 = Set Skel_Mesh(Skel_Mesh=GetOwnerSkeletalMeshComponent())
   > Is Valid(InputObject=@n1)
   > Set Anim_Inst(Anim_Inst:Anim Instance Object Reference=GetAnimInstance(self=@n1)) -> (Output_Get:Anim Instance Object Reference)
   Event BlueprintUpdateAnimation()
   > n2 = Sequence!(execute) -> (then_0, then_1, then_2:Exec)
   n3 = Get Anim_Inst() -> (Anim_Inst:Anim Instance Object Reference)
   > Is Valid(InputObject=@n3)
   > Set idlePlaying(idlePlaying:double="To Float (Boolean)"(InBool=Montage_IsPlaying(self=@n3, Montage=/BRCosmetics/Animation/Game/MainPlayer/Menu/BR/CustomSkins/_s21/BlueJaySummerF/BlueJaySummer_Female_Idle_Montage))) -> (Output_Get:double)
   n4 = Gravity Override State Function DEPRECATED(
    execute=@n2.then_1
    CharacterPartAnimInstance=Self-Reference()
    GameTimeInSeconds=GetGameTimeInSeconds()
    GravityOverrideParams=Get Gravity Override Params()
   )
   > Set gravityVector(gravityVector:Vector=@n4) -> (Output_Get:Vector)
   n5 = Branch(
    execute=@n2.then_2
    Condition=BackpackItemDefinitionHasAnyMetaTag!(Tags:Array of Gameplay Tag Structures=Make Array("[0]"=(TagName="Cosmetics.MetaTags.EnsembleMaroon")), bRequiresEquipped:bool=false) -> (ReturnValue:bool)
   )
   > Set EnsembleBackpack(EnsembleBackpack:bool=true) -> (Output_Get:bool)
   Set EnsembleBackpack(execute=@n5.else, EnsembleBackpack=false)
  ]}
]
variables [
 Anim_Inst {type=AnimInstance, default=null}
 EnsembleBackpack {type=boolean}
 Gravity Override Params {
  type=GravityOverrideParamsStruct
  default {skydive_DivingDown {gravityOverride=<-1500 0 300>}, skydive_FloatingDown {gravityOverride=<0 -700 300>}}
 }
 gravityVector {type=Vector}
 idlePlaying {type=number}
 Skel_Mesh {type=SkeletalMeshComponent}
]
